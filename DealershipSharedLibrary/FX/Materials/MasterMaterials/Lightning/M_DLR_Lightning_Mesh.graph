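# Translucent unlit master material for lightning on Niagara mesh particles, with flickering opacity and adjustable emissive color.
# Three multiplied sine waves, offset by dynamic or optional particle random values and scaled by FlickerSpeed, are contrast-boosted to form the flicker.
# Flicker multiplies particle alpha and an optional, invertible Fresnel, then an optional soft ramp mask, before depth fade.
# Particle color feeds MF_FX_EmissiveAdjustments; the Erode and VerticalTile dynamic parameters are unused.
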Material /DealershipSharedLibrary/FX/Materials/MasterMaterials/Lightning/M_DLR_Lightning_Mesh
graph {
 properties {
  bAutomaticallySetUsageInEditor=false
  bUsedWithNiagaraMeshParticles=true
  blendMode=BLEND_Translucent
  shadingModel=MSM_Unlit
  usedShadingModels=Unlit
 }
 nodes [
  n1 = LinearInterpolate(A=ParticleRandom(), B=Constant(), Alpha=ScalarParameter() {parameterName=Use Particle Random Value})
  n2 = DynamicParameter() {defaultValue=<rgba 1 1 0 1>, paramNames [Random,FlickerSpeed,Erode,VerticalTile]}
  n3 = Multiply(A=Add(A=Time(), B=Add(A=@n1, B=@n2.Random)), B=@n2.FlickerSpeed)
  n4 = Sine(None=@n3) {period=0.11}
  n5 = Multiply(A=Sine(None=@n3) {period=0.09}, B=Sine(None=@n3) {period=0.075})
  n6 = MaterialFunctionCall(In=ConstantBiasScale(None=Multiply(A=@n4, B=@n5)), Contrast=ScalarParameter() {defaultValue=3, parameterName=FlickerContrast}) {materialFunction=&CheapContrast}
  n7 = ParticleColor()
  n8 = Fresnel(
   ExponentIn=ScalarParameter() {defaultValue=5, parameterName=ExponentIn}
   BaseReflectFractionIn=ScalarParameter() {defaultValue=0.5, parameterName=BaseReflectFractionIn}
  ) {baseReflectFraction=0.5}
  n9 = LinearInterpolate(A=@n8, B=OneMinus(None=@n8), Alpha=ScalarParameter() {parameterName=Invert Fresnel})
  n10 = ScalarParameter() {defaultValue=1, parameterName=Fresnel Power}
  n11 = Multiply(
   A=Multiply(A=Saturate(None=@n6.Result), B=Clamp(None=@n7.A) {clampMode=CMODE_ClampMax})
   B=LinearInterpolate(B=Power(Base=@n9, Exp=@n10), Alpha=ScalarParameter() {defaultValue=1, parameterName="UseFresnel(0-1)"}) {constA=1}
  )
  n12 = Constant() {r=1}
  n13 = ScalarParameter() {defaultValue=1, parameterName=OpacityMaskTileY}
  n14 = TextureSample(UVs=Multiply(A=TextureCoordinate(), B=AppendVector(A=@n12, B=@n13))) {samplerSource=SSM_Clamp_WorldGroupSettings, texture=/Game/Effects/Fort_Effects/Textures/Ramps/T_Ramp_Radial_Soft}
  DepthFade(
   Opacity=If(A=ScalarParameter() {parameterName=UseOpacityMask}, "A > B"=Multiply(A=@n11, B=@n14.RGB), "A == B"=@n11, "A < B"=@n11)
   FadeDistance=ScalarParameter() {defaultValue=745, parameterName=DepthFade}
  )
  MaterialFunctionCall(Emissive=@n7.RGB) {materialFunction=&MF_FX_EmissiveAdjustments}
  Reroute(None=@n11)
 ]
}
