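# Translucent unlit, two-sided particle material for stretching electric lightning arcs, usable on mesh and sprite particles.
# Vertex-offset math combines SplineThicken width, masked sine wobble, and a pull of one UV end toward a randomized point near the actor position, easing off late in particle life.
# A panning plasma texture blends edge and core colors into HDR emissive and drives opacity, faded at both V ends.
# Dynamic parameters supply per-particle width, wobble, randomness, pan timing and UV offset.
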Material /DealershipSharedLibrary/FX/Materials/MasterMaterials/Lightning/M_Lightning_Arc_Stretch_WPO
graph {
 properties {
  bUsedWithMeshParticles=true
  bUsedWithNiagaraMeshParticles=true
  bUsedWithNiagaraSprites=true
  bUsedWithParticleSprites=true
  blendMode=BLEND_Translucent
  pixelDepthOffsetMode=PDOM_Legacy
  shadingModel=MSM_Unlit
  twoSided=true
  usedShadingModels=Unlit
 }
 nodes [
  n1 = DynamicParameter() {paramNames [Width,WobbleMult,Random1,Random2]}
  n2 = Multiply(A=ScalarParameter() {defaultValue=6, parameterName=Width}, B=@n1.Width)
  n3 = ScalarParameter() {parameterName=WidthTipMult}
  n4 = TextureCoordinate()
  n5 = Constant() {r=0.5}
  n6 = Constant()
  n7 = TextureCoordinate()
  n8 = Add(A=@n1.Random2, B=Time())
  n9 = Multiply(A=@n8, B=ScalarParameter() {defaultValue=3, parameterName=WobbleSpeed})
  n10 = ComponentMask(None=Add(A=@n7, B=Multiply(A=@n9) {constB=-0.6})) {g=true}
  n11 = ComponentMask(None=Add(A=@n7, B=Multiply(A=@n9) {constB=-1})) {g=true}
  n12 = Add(A=Multiply(A=Sine(None=@n10) {period=0.2}) {constB=0.2}, B=Sine(None=@n11) {period=0.8})
  n13 = Multiply(A=Sine(None=@n10) {period=0.23}) {constB=0.2}
  n14 = Sine(None=@n11) {period=0.6}
  n15 = Transform(None=AppendVector(A=AppendVector(A=@n6, B=@n12), B=Add(A=@n13, B=@n14))) {transformSourceType=TRANSFORMSOURCE_Local}
  n16 = ScalarParameter() {defaultValue=5, parameterName=Wobble}
  n17 = Multiply(
   A=SphereMask(A=ComponentMask(None=@n4) {g=true}, B=@n5, Radius=@n5) {hardnessPercent=50}
   B=Multiply(A=Normalize(VectorInput=@n15), B=Multiply(A=@n16, B=@n1.WobbleMult))
  )
  Add(
   A=MaterialFunctionCall(WidthBase=@n2, WidthTip=Multiply(A=@n2, B=@n3) {constB=5}, UVs for Thickness=Power(Base=@n4) {constExponent=0.6}) {materialFunction=&SplineThicken}.WorldPositionOffset
   B=@n17
  )
  n18 = Multiply(
   A=Subtract(A=AppendVector(A=@n1.Random1, B=@n1.Random1)) {constB=0.5}
   B=ScalarParameter() {defaultValue=30, parameterName=ActorPositionOffset}
  ) {constB=10}
  n19 = Constant()
  n20 = Subtract(A=Add(A=ActorPositionWS(), B=AppendVector(A=@n18, B=@n19)), B=WorldPosition().XYZ)
  n21 = OneMinus(None=ComponentMask(None=TextureCoordinate()) {g=true})
  n22 = LinearInterpolate(Alpha=ParticleRelativeTime()) {constA=2, constB=0}
  Add(A=@n17, B=Multiply(A=Multiply(A=@n20, B=@n21), B=Clamp(None=@n22)))
  n23 = ParticleColor()
  n24 = VectorParameter() {parameterName=EdgeColor, defaultValue=<rgba 0.2226 0.2425 0.5 0>}
  n25 = VectorParameter() {parameterName=Color, defaultValue=<rgba 0.6214 0.9501 1 0>}
  n26 = DynamicParameter() {parameterIndex=1, defaultValue=<rgba 1 0 0 0>, paramNames [Time Multiplier, UV Offset V, Param3, Param4]}
  n27 = Add(A=TextureCoordinate() {vTiling=0.5}, B=AppendVector(A=Constant(), B=@n26.UV Offset V))
  n28 = AppendVector(A=Constant() {r=1}, B=LinearInterpolate(Alpha=@n1.Random1) {constA=0.5, constB=1.2})
  n29 = Multiply(A=ScalarParameter() {defaultValue=1, parameterName=PanSpeed}, B=@n26.Time Multiplier)
  n30 = ScalarParameter() {parameterName=Speed_X}
  n31 = ScalarParameter() {parameterName=Speed_Y}
  n32 = TextureSampleParameter2D(UVs=Panner(Coordinate=Multiply(A=@n27, B=@n28), Time=Multiply(A=@n8, B=@n29), Speed=AppendVector(A=@n30, B=@n31)) {speedY=-1}) {parameterName=Texture, texture=/Game/Effects/Fort_Effects/Textures/Tile/T_Lightning_Plasma_Soft}
  n33 = Multiply(A=@n23.RGB, B=Desaturation(None=LinearInterpolate(A=@n24.RGB, B=@n25.RGB, Alpha=@n32.G)))
  MaterialFunctionCall(Emissive=MaterialFunctionCall(LDR Input=Clamp(None=@n33), HDR Tint=@n33, HDR Intensity Multiplier=ScalarParameter() {defaultValue=20, parameterName=HDRIntensity}) {materialFunction=&DeriveHDRfromLDR}.RemappedRange) {materialFunction=&MF_FX_EmissiveAdjustments}
  n34 = ComponentMask(None=TextureCoordinate()) {g=true}
  n35 = Multiply(A=@n32.B) {constB=0.2}
  Multiply(A=Multiply(A=@n34, B=OneMinus(None=@n34)), B=Multiply(A=Add(A=@n32.R, B=@n35), B=@n23.A))
 ]
}
